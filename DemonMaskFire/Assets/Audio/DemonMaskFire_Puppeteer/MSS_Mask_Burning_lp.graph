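# Stereo MetaSound source for the Demon Mask's burning fire loop, using a temporary wave asset, boss-fight sound class, aggro FX attenuation and concurrency override.
# On play, it picks a random wave from the array, randomizes pitch shift, then starts the wave player.
# An ADSR envelope triggered on play multiplies both channels for a soft attack.
# The StopFireball input is unconnected, and OnFinished comes only from the wave player.
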MetaSoundSource /DemonMaskFire/Assets/Audio/DemonMaskFire_Puppeteer/MSS_Mask_Burning_lp
inputs [
 StopFireball {type=Trigger}
 UE.Source.OnPlay {type=Trigger}
 Wave Asset {type=WaveAsset:Array}
]
interfaces [UE.OutputFormat.Stereo:1.0, UE.Source.OneShot:1.0, UE.Source:1.1]
outputs [
 "UE.OutputFormat.Stereo.Audio:0" {type=Audio}
 "UE.OutputFormat.Stereo.Audio:1" {type=Audio}
 UE.Source.OneShot.OnFinished {type=Trigger}
]
pages [Default {
  input_defaults {StopFireball=null, UE.Source.OnPlay=null, Wave Asset [/DemonMaskFire/Assets/Audio/DemonMaskFire_Puppeteer/TEMP/Temp_Fire_loop]}
  nodes [
   ADSREnvelopeADSREnvelope_0 = ADSR Envelope.ADSR Envelope:Audio(Trigger Attack=^UE.Source.OnPlay) {also {defaults {Attack Time=0.6, Decay Time=0.3}}}
   ArrayRandomGet_0 = Array.Random Get:WaveAsset:Array(Next=^UE.Source.OnPlay, In Array=^Wave Asset)
   n1 = UE.RandomFloat(Next=@ArrayRandomGet_0.On Next) {also {defaults {Max=3, Min=-3}}}
   WavePlayer_0 = UE.Wave Player:Stereo(Wave Asset=@ArrayRandomGet_0.Value, Play=@n1.On Next, Pitch Shift=@n1.Value) {also {defaults {Loop="<MetaSoundBuilderResult.SUCCEEDED: 0>"}}}
   Multiply_0 = UE.Multiply:Audio(PrimaryOperand=@ADSREnvelopeADSREnvelope_0.Out Envelope, AdditionalOperands=@WavePlayer_0.Out Left)
   Multiply_1 = UE.Multiply:Audio(PrimaryOperand=@ADSREnvelopeADSREnvelope_0.Out Envelope, AdditionalOperands=@WavePlayer_0.Out Right)
  ]
  links [
   {dst="@UE.OutputFormat.Stereo.Audio:0", src=Multiply_0.Out}
   {dst="@UE.OutputFormat.Stereo.Audio:1", src=Multiply_1.Out}
   {dst="@UE.Source.OneShot.OnFinished", src=WavePlayer_0.On Finished}
  ]
 }]
properties {
 assetImportData=&self:AssetImportData
 attenuationSettings=&/SA_DemonMask_AggroFx
 bOverrideConcurrency=true
 maxDistance=6000
 numChannels=2
 outputFormat=Stereo
 sampleRate=48000
 soundClassObject=&Fort_BossFight
 volume=1.5
 concurrencyOverrides {maxCount=3, resolutionRule=StopOldest, volumeScale=0.5}
}
