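# Stereo one-shot MetaSound for the close pass-by of the Puppeteer demon mask fireball, routed to the boss-fight projectile sound class.
# On play, a random fireball wave starts immediately and a random scream layer follows after a short random delay.
# Both players are mixed to stereo, and a trigger accumulator signals finished once both end.
# The BulletWhip interface is present, but none of its inputs drive the graph.
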MetaSoundSource /DemonMaskFire/Assets/Audio/DemonMaskFire_Puppeteer/MSS_Mask_Proj_By_Close
inputs [
 BulletWhip.OnStop {type=Trigger}
 BulletWhip.PassByAlignment {type=Float}
 BulletWhip.PassByAzimuth {type=Float}
 BulletWhip.PassByDistance {type=Float}
 Layer 2 {type=WaveAsset:Array}
 UE.Source.OnPlay {type=Trigger}
 Wave Asset {type=WaveAsset:Array}
]
interfaces [Fort.BulletWhip:1.0, UE.OutputFormat.Stereo:1.0, UE.Source.OneShot:1.0, UE.Source:1.1]
outputs [
 "UE.OutputFormat.Stereo.Audio:0" {type=Audio}
 "UE.OutputFormat.Stereo.Audio:1" {type=Audio}
 UE.Source.OneShot.OnFinished {type=Trigger}
]
pages [Default {
  input_defaults {
   BulletWhip.OnStop=null
   BulletWhip.PassByAlignment=0
   BulletWhip.PassByAzimuth=-90
   BulletWhip.PassByDistance=1
   UE.Source.OnPlay=null
   Layer 2 [
    /DemonMaskFire/Assets/Audio/DemonMaskFire_Puppeteer/wav/Puppeteer_Mask_Fireball_By_Close_Scream_01
    /DemonMaskFire/Assets/Audio/DemonMaskFire_Puppeteer/wav/Puppeteer_Mask_Fireball_By_Close_Scream_02
    /DemonMaskFire/Assets/Audio/DemonMaskFire_Puppeteer/wav/Puppeteer_Mask_Fireball_By_Close_Scream_03
    /DemonMaskFire/Assets/Audio/DemonMaskFire_Puppeteer/wav/Puppeteer_Mask_Fireball_By_Close_Scream_04
   ]
   Wave Asset [
    /DemonMaskFire/Assets/Audio/DemonMaskFire_Puppeteer/wav/Puppeteer_Mask_Fireball_By_Close_01
    /DemonMaskFire/Assets/Audio/DemonMaskFire_Puppeteer/wav/Puppeteer_Mask_Fireball_By_Close_02
    /DemonMaskFire/Assets/Audio/DemonMaskFire_Puppeteer/wav/Puppeteer_Mask_Fireball_By_Close_03
    /DemonMaskFire/Assets/Audio/DemonMaskFire_Puppeteer/wav/Puppeteer_Mask_Fireball_By_Close_04
   ]
  }
  nodes [
   n1 = UE.RandomTime(Next=^UE.Source.OnPlay) {also {defaults {Max=0.4, Min=0.2}}}
   TriggerDelay_0 = UE.Trigger Delay(In=@n1.On Next, Delay Time=@n1.Value)
   ArrayRandomGet_1 = Array.Random Get:WaveAsset:Array(In Array=^Layer 2, Next=@TriggerDelay_0.Out) {also {defaults {Weights []}}}
   WavePlayer_0 = UE.Wave Player:Stereo(Play=@ArrayRandomGet_1.On Next, Wave Asset=@ArrayRandomGet_1.Value)
   ArrayRandomGet_0 = Array.Random Get:WaveAsset:Array(Next=^UE.Source.OnPlay, In Array=^Wave Asset)
   WavePlayer_1 = UE.Wave Player:Stereo(Play=@ArrayRandomGet_0.On Next, Wave Asset=@ArrayRandomGet_0.Value)
   AudioMixerAudioMixerStereo2_0 = "AudioMixer.Audio Mixer (Stereo, 2)"(In 1 L=@WavePlayer_0.Out Left, In 1 R=@WavePlayer_0.Out Right, In 0 L=@WavePlayer_1.Out Left, In 0 R=@WavePlayer_1.Out Right)
   TriggerAccumulatorTriggerAccumulate2_0 = "TriggerAccumulator.Trigger Accumulate (2)"(In 1=@WavePlayer_0.On Finished, In 0=@WavePlayer_1.On Finished)
  ]
  links [
   {dst="@UE.OutputFormat.Stereo.Audio:0", src=AudioMixerAudioMixerStereo2_0.Out L}
   {dst="@UE.OutputFormat.Stereo.Audio:1", src=AudioMixerAudioMixerStereo2_0.Out R}
   {dst="@UE.Source.OneShot.OnFinished", src=TriggerAccumulatorTriggerAccumulate2_0.Out}
  ]
 }]
properties {
 assetImportData=&self:AssetImportData
 attenuationSettings=&/SA_DemonMask_Fireball
 bBypassVolumeScaleForPriority=true
 maxDistance=7000
 numChannels=2
 outputFormat=Stereo
 sampleRate=48000
 soundClassObject=/Game/Sounds/FortSoundClassesAndModes/Fort_BossFight_SoundClass/Fort_BossFight_Projectile
 concurrencyOverrides {maxCount=4, resolutionRule=StopOldest, volumeScale=0.6}
 concurrencySet [&/CS_DemonMask_Fireball_Proj]
 soundSubmixSends [{maxSendLevel=0.2, minSendLevel=0.2, sendLevel=0.2, soundSubmix=&Fortnite_Reverb_Distant}]
}
