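# Stereo one-shot MetaSound Source for the "By_Dist" fireball projectile sound of the DemonMaskFire Puppeteer mask, routed to the boss-fight projectile sound class.
# On play, a random array get picks one of four fireball wave variations and feeds a stereo wave player.
# The player drives both outputs and signals finish when the wave ends.
# Fireball-specific attenuation and concurrency assets govern playback, alongside stop-oldest concurrency override values.
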MetaSoundSource /DemonMaskFire/Assets/Audio/DemonMaskFire_Puppeteer/MSS_Mask_Proj_By_Dist
inputs [
 UE.Source.OnPlay {type=Trigger}
 Wave Asset {type=WaveAsset:Array}
]
interfaces [UE.OutputFormat.Stereo:1.0, UE.Source.OneShot:1.0, UE.Source:1.1]
outputs [
 "UE.OutputFormat.Stereo.Audio:0" {type=Audio}
 "UE.OutputFormat.Stereo.Audio:1" {type=Audio}
 UE.Source.OneShot.OnFinished {type=Trigger}
]
pages [Default {
  input_defaults {
   UE.Source.OnPlay=null
   Wave Asset [
    /DemonMaskFire/Assets/Audio/DemonMaskFire_Puppeteer/wav/Puppeteer_Mask_Fireball_By_Dist_01
    /DemonMaskFire/Assets/Audio/DemonMaskFire_Puppeteer/wav/Puppeteer_Mask_Fireball_By_Dist_02
    /DemonMaskFire/Assets/Audio/DemonMaskFire_Puppeteer/wav/Puppeteer_Mask_Fireball_By_Dist_03
    /DemonMaskFire/Assets/Audio/DemonMaskFire_Puppeteer/wav/Puppeteer_Mask_Fireball_By_Dist_04
   ]
  }
  nodes [
   ArrayRandomGet_0 = Array.Random Get:WaveAsset:Array(Next=^UE.Source.OnPlay, In Array=^Wave Asset)
   WavePlayer_0 = UE.Wave Player:Stereo(Play=@ArrayRandomGet_0.On Next, Wave Asset=@ArrayRandomGet_0.Value)
  ]
  links [
   {dst="@UE.Source.OneShot.OnFinished", src=WavePlayer_0.On Finished}
   {dst="@UE.OutputFormat.Stereo.Audio:0", src=WavePlayer_0.Out Left}
   {dst="@UE.OutputFormat.Stereo.Audio:1", src=WavePlayer_0.Out Right}
  ]
 }]
properties {
 assetImportData=&self:AssetImportData
 attenuationSettings=&/SA_DemonMask_Fireball
 bBypassVolumeScaleForPriority=true
 maxDistance=7000
 numChannels=2
 outputFormat=Stereo
 priority=0.9
 sampleRate=48000
 soundClassObject=/Game/Sounds/FortSoundClassesAndModes/Fort_BossFight_SoundClass/Fort_BossFight_Projectile
 volume=0.4
 concurrencyOverrides {maxCount=2, resolutionRule=StopOldest, retriggerTime=0.2, volumeScale=0.6}
 concurrencySet [&/CS_DemonMask_Fireball_Proj]
}
