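# One-shot stereo MetaSound for the Puppeteer Demon Mask fire projectile impact.
# On play, it mixes a randomly chosen impact wave with a fixed temp fireball layer, each with its own random pitch shift.
# The random impact player alone signals the end of the sound.
# Distance-scaled pre-effect sends feed ambient and music sidechain buses, and a concurrency override stops the oldest instance beyond a small limit.
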MetaSoundSource /DemonMaskFire/Assets/Audio/DemonMaskFire_Puppeteer/MSS_Mask_Proj_Impact
inputs [
 Additional Layer {type=WaveAsset}
 UE.Source.OnPlay {type=Trigger}
 Wave Asset {type=WaveAsset:Array}
]
interfaces [UE.OutputFormat.Stereo:1.0, UE.Source.OneShot:1.0, UE.Source:1.1]
outputs [
 "UE.OutputFormat.Stereo.Audio:0" {type=Audio}
 "UE.OutputFormat.Stereo.Audio:1" {type=Audio}
 UE.Source.OneShot.OnFinished {type=Trigger}
]
pages [Default {
  input_defaults {
   Additional Layer=/DemonMaskFire/Assets/Audio/DemonMaskFire_Puppeteer/TEMP/Temp_Fireball_Impt_layer
   UE.Source.OnPlay=null
   Wave Asset [
    /DemonMaskFire/Assets/Audio/DemonMaskFire_Puppeteer/wav/Puppeteer_Mask_Fire_Impt_01
    /DemonMaskFire/Assets/Audio/DemonMaskFire_Puppeteer/wav/Puppeteer_Mask_Fire_Impt_02
    /DemonMaskFire/Assets/Audio/DemonMaskFire_Puppeteer/wav/Puppeteer_Mask_Fire_Impt_03
    /DemonMaskFire/Assets/Audio/DemonMaskFire_Puppeteer/wav/Puppeteer_Mask_Fire_Impt_04
    /DemonMaskFire/Assets/Audio/DemonMaskFire_Puppeteer/wav/Puppeteer_Mask_Fire_Impt_05
    /DemonMaskFire/Assets/Audio/DemonMaskFire_Puppeteer/wav/Puppeteer_Mask_Fire_Impt_06
   ]
  }
  nodes [
   n1 = UE.RandomFloat(Next=UE.Reroute(Value=^UE.Source.OnPlay).Value) {also {defaults {Min=1.3}}}
   WavePlayer_0 = UE.Wave Player:Stereo(Wave Asset=^Additional Layer, Play=@n1.On Next, Pitch Shift=@n1.Value)
   ArrayRandomGet_0 = Array.Random Get:WaveAsset:Array(Next=^UE.Source.OnPlay, In Array=^Wave Asset)
   n2 = UE.RandomFloat(Next=@ArrayRandomGet_0.On Next) {also {defaults {Max=2, Min=1}}}
   WavePlayer_1 = UE.Wave Player:Stereo(Wave Asset=@ArrayRandomGet_0.Value, Play=@n2.On Next, Pitch Shift=@n2.Value)
   AudioMixerAudioMixerStereo2_0 = "AudioMixer.Audio Mixer (Stereo, 2)"(In 1 L=@WavePlayer_0.Out Left, In 1 R=@WavePlayer_0.Out Right, In 0 L=@WavePlayer_1.Out Left, In 0 R=@WavePlayer_1.Out Right) {also {defaults {Gain 0=0.7}}}
  ]
  links [
   {dst="@UE.OutputFormat.Stereo.Audio:0", src=AudioMixerAudioMixerStereo2_0.Out L}
   {dst="@UE.OutputFormat.Stereo.Audio:1", src=AudioMixerAudioMixerStereo2_0.Out R}
   {dst="@UE.Source.OneShot.OnFinished", src=WavePlayer_1.On Finished}
  ]
 }]
properties {
 assetImportData=&self:AssetImportData
 attenuationSettings=&/SA_DemonMask_Fireball_Impact
 bOverrideConcurrency=true
 maxDistance=4500
 numChannels=2
 outputFormat=Stereo
 sampleRate=48000
 soundClassObject=&Fort_Impacts
 volume=1.4
 concurrencyOverrides {maxCount=3, resolutionRule=StopOldest, volumeScale=0.75}
 modulationSettings {volumeModulationDestination {bEnableModulation=true}}
 preEffectBusSends [
  {
   audioBus=&AB_Sidechain_FortAmbient_Vol
   maxSendDistance=3000
   maxSendLevel=0
   minSendDistance=0
   minSendLevel=0.5
   sendLevel=0
   sourceBusSendLevelControlMethod=Linear
  }
  {
   audioBus=&AB_Sidechain_FortMusic_Vol
   maxSendDistance=3000
   maxSendLevel=0
   minSendDistance=0
   minSendLevel=0.3
   sendLevel=0
   sourceBusSendLevelControlMethod=Linear
  }
  {
   audioBus=/Game/Sounds/Modulation/EnvelopeFollowers/LPF/AB_SideChain_FortMusic_LPF
   maxSendDistance=4000
   maxSendLevel=0
   minSendDistance=0
   minSendLevel=0.3
   sendLevel=0
   sourceBusSendLevelControlMethod=Linear
  }
  {
   audioBus=/Game/Sounds/Modulation/EnvelopeFollowers/HPF/AB_SideChain_FortMusic_HPF
   maxSendDistance=4000
   maxSendLevel=0
   minSendDistance=0
   minSendLevel=0.4
   sendLevel=0
   sourceBusSendLevelControlMethod=Linear
  }
 ]
}
