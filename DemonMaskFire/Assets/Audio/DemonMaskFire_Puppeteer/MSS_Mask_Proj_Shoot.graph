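# One-shot stereo MetaSound for the Puppeteer demon mask's fireball shot.
# On play, it picks one of five shoot waves, then rolls a random pitch shift and a slight random gain that scales both channels.
# Voices are capped with stop-oldest concurrency and spatialized through the shoot attenuation asset.
# A distance-scaled pre-effect send feeds the music volume sidechain bus, suggesting nearby shots duck the music.
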MetaSoundSource /DemonMaskFire/Assets/Audio/DemonMaskFire_Puppeteer/MSS_Mask_Proj_Shoot
inputs [
 UE.Source.OnPlay {type=Trigger}
 Wave Asset {type=WaveAsset:Array}
]
interfaces [UE.OutputFormat.Stereo:1.0, UE.Source.OneShot:1.0, UE.Source:1.1]
outputs [
 "UE.OutputFormat.Stereo.Audio:0" {type=Audio}
 "UE.OutputFormat.Stereo.Audio:1" {type=Audio}
 UE.Source.OneShot.OnFinished {type=Trigger}
]
pages [Default {
  input_defaults {
   UE.Source.OnPlay=null
   Wave Asset [
    /DemonMaskFire/Assets/Audio/DemonMaskFire_Puppeteer/wav/Puppeteer_Mask_Fireball_Shoot_01
    /DemonMaskFire/Assets/Audio/DemonMaskFire_Puppeteer/wav/Puppeteer_Mask_Fireball_Shoot_02
    /DemonMaskFire/Assets/Audio/DemonMaskFire_Puppeteer/wav/Puppeteer_Mask_Fireball_Shoot_03
    /DemonMaskFire/Assets/Audio/DemonMaskFire_Puppeteer/wav/Puppeteer_Mask_Fireball_Shoot_04
    /DemonMaskFire/Assets/Audio/DemonMaskFire_Puppeteer/wav/Puppeteer_Mask_Fireball_Shoot_05
   ]
  }
  nodes [
   ArrayRandomGet_0 = Array.Random Get:WaveAsset:Array(Next=^UE.Source.OnPlay, In Array=^Wave Asset) {also {defaults {Shared State Behavior=2}}}
   n1 = UE.RandomFloat(Next=@ArrayRandomGet_0.On Next)
   n2 = UE.RandomFloat(Next=@n1.On Next) {also {defaults {Max=1.2, Min=1}}}
   WavePlayer_0 = UE.Wave Player:Stereo(Wave Asset=@ArrayRandomGet_0.Value, Play=@n2.On Next, Pitch Shift=@n1.Value)
   Multiply_0 = UE.Multiply:Audio by Float(AdditionalOperands=@n2.Value, PrimaryOperand=@WavePlayer_0.Out Left)
   Multiply_1 = UE.Multiply:Audio by Float(AdditionalOperands=@n2.Value, PrimaryOperand=@WavePlayer_0.Out Right)
  ]
  links [
   {dst="@UE.OutputFormat.Stereo.Audio:0", src=Multiply_0.Out}
   {dst="@UE.OutputFormat.Stereo.Audio:1", src=Multiply_1.Out}
   {dst="@UE.Source.OneShot.OnFinished", src=WavePlayer_0.On Finished}
  ]
 }]
properties {
 assetImportData=&self:AssetImportData
 attenuationSettings=&/SA_DemonMask_Shoot
 bOverrideConcurrency=true
 maxDistance=14000
 numChannels=2
 outputFormat=Stereo
 sampleRate=48000
 soundClassObject=&Fort_BossFight
 concurrencyOverrides {bVolumeScaleCanRelease=true, maxCount=6, resolutionRule=StopOldest, voiceStealReleaseTime=0.1, volumeScaleReleaseTime=0.1}
 preEffectBusSends [{
   audioBus=&AB_Sidechain_FortMusic_Vol
   maxSendDistance=6000
   maxSendLevel=0
   minSendDistance=0
   minSendLevel=0.3
   sendLevel=0
   sourceBusSendLevelControlMethod=Linear
  }]
}
