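# Stereo MetaSound source for the Puppeteer demon mask fireball projectile loop, routed to the boss-fight projectile sound class.
# On play, it picks one of three loop waves at random and starts it with a randomized pitch shift.
# An ADSR envelope multiplies both channels, attacking on play and releasing on the StopFireball trigger.
# OnFinished comes from the wave player rather than the envelope; attenuation, concurrency limits and a distant reverb send are configured.
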MetaSoundSource /DemonMaskFire/Assets/Audio/DemonMaskFire_Puppeteer/MSS_Mask_Proj_loop
inputs [
 StopFireball {type=Trigger}
 UE.Source.OnPlay {type=Trigger}
 Wave Asset {type=WaveAsset:Array}
]
interfaces [UE.OutputFormat.Stereo:1.0, UE.Source.OneShot:1.0, UE.Source:1.1]
outputs [
 "UE.OutputFormat.Stereo.Audio:0" {type=Audio}
 "UE.OutputFormat.Stereo.Audio:1" {type=Audio}
 UE.Source.OneShot.OnFinished {type=Trigger}
]
pages [Default {
  input_defaults {
   StopFireball=null
   UE.Source.OnPlay=null
   Wave Asset [
    /DemonMaskFire/Assets/Audio/DemonMaskFire_Puppeteer/wav/Puppeteer_Mask_Fireball_lp_01
    /DemonMaskFire/Assets/Audio/DemonMaskFire_Puppeteer/wav/Puppeteer_Mask_Fireball_lp_02
    /DemonMaskFire/Assets/Audio/DemonMaskFire_Puppeteer/wav/Puppeteer_Mask_Fireball_lp_03
   ]
  }
  nodes [
   ADSREnvelopeADSREnvelope_0 = ADSR Envelope.ADSR Envelope:Audio(Trigger Release=^StopFireball, Trigger Attack=^UE.Source.OnPlay) {also {defaults {Attack Time=0.15, Hard Reset="<MetaSoundBuilderResult.SUCCEEDED: 0>", Release Time=0.1, Sustain Level=0.9}}}
   ArrayRandomGet_0 = Array.Random Get:WaveAsset:Array(Next=^UE.Source.OnPlay, In Array=^Wave Asset)
   n1 = UE.RandomFloat(Next=@ArrayRandomGet_0.On Next) {also {defaults {Min=-1}}}
   WavePlayer_0 = UE.Wave Player:Stereo(Wave Asset=@ArrayRandomGet_0.Value, Play=@n1.On Next, Pitch Shift=@n1.Value) {also {defaults {Loop="<MetaSoundBuilderResult.SUCCEEDED: 0>"}}}
   Multiply_0 = UE.Multiply:Audio(PrimaryOperand=@ADSREnvelopeADSREnvelope_0.Out Envelope, AdditionalOperands=@WavePlayer_0.Out Left)
   Multiply_1 = UE.Multiply:Audio(PrimaryOperand=@ADSREnvelopeADSREnvelope_0.Out Envelope, AdditionalOperands=@WavePlayer_0.Out Right)
  ]
  links [
   {dst="@UE.OutputFormat.Stereo.Audio:0", src=Multiply_0.Out}
   {dst="@UE.OutputFormat.Stereo.Audio:1", src=Multiply_1.Out}
   {dst="@UE.Source.OneShot.OnFinished", src=WavePlayer_0.On Finished}
  ]
 }]
properties {
 assetImportData=&self:AssetImportData
 attenuationSettings=&/SA_DemonMask_Fireball
 bBypassVolumeScaleForPriority=true
 maxDistance=7000
 numChannels=2
 outputFormat=Stereo
 priority=1.1
 sampleRate=48000
 soundClassObject=/Game/Sounds/FortSoundClassesAndModes/Fort_BossFight_SoundClass/Fort_BossFight_Projectile
 concurrencyOverrides {maxCount=3, resolutionRule=StopOldest, volumeScale=0.5}
 concurrencySet [&/CS_DemonMask_Fireball_Proj]
 modulationSettings {volumeModulationDestination {bEnableModulation=true, modulators [null]}}
 soundSubmixSends [{
   maxSendDistance=4000
   maxSendLevel=0.4
   minSendDistance=0
   minSendLevel=0.15
   sendLevel=0.15
   soundSubmix=&Fortnite_Reverb_Distant
  }]
}
