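# One-shot stereo MetaSound source for the Puppeteer mask destroyed sound in DemonMaskFire.
# On play, a random-get node picks from the wave array, which currently holds one asset, and a stereo wave player plays it and signals finish.
# It routes to the Fort_BossFight sound class and borrows the PuppeteerMask Awaken attenuation.
# A concurrency override limits simultaneous instances, stopping the oldest.
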MetaSoundSource /DemonMaskFire/Assets/Audio/DemonMaskFire_Puppeteer/MSS_Puppeteer_Mask_Destroyed
inputs [
 UE.Source.OnPlay {type=Trigger}
 Wave Asset {type=WaveAsset:Array}
]
interfaces [UE.OutputFormat.Stereo:1.0, UE.Source.OneShot:1.0, UE.Source:1.1]
outputs [
 "UE.OutputFormat.Stereo.Audio:0" {type=Audio}
 "UE.OutputFormat.Stereo.Audio:1" {type=Audio}
 UE.Source.OneShot.OnFinished {type=Trigger}
]
pages [Default {
  input_defaults {UE.Source.OnPlay=null, Wave Asset [/DemonMaskFire/Assets/Audio/DemonMaskFire_Puppeteer/wav/Puppeteer_Mask_Destroyed]}
  nodes [
   ArrayRandomGet_0 = Array.Random Get:WaveAsset:Array(Next=^UE.Source.OnPlay, In Array=^Wave Asset)
   WavePlayer_0 = UE.Wave Player:Stereo(Play=@ArrayRandomGet_0.On Next, Wave Asset=@ArrayRandomGet_0.Value)
  ]
  links [
   {dst="@UE.Source.OneShot.OnFinished", src=WavePlayer_0.On Finished}
   {dst="@UE.OutputFormat.Stereo.Audio:0", src=WavePlayer_0.Out Left}
   {dst="@UE.OutputFormat.Stereo.Audio:1", src=WavePlayer_0.Out Right}
  ]
 }]
properties {
 assetImportData=&self:AssetImportData
 attenuationSettings=&/Att_PuppeteerMask_Awaken
 bOverrideConcurrency=true
 maxDistance=10200
 numChannels=2
 outputFormat=Stereo
 sampleRate=48000
 soundClassObject=&Fort_BossFight
 volume=1.4
 concurrencyOverrides {maxCount=3, resolutionRule=StopOldest, volumeScale=0.5}
}
